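# Character-parent-based material for the SoloSnooze backpack cosmetic, usable on skeletal meshes with morph targets and mesh deformers.
# The 2023 Character Parent PreFX function supplies base material attributes, masks, emissive UVs, Fresnel and position data through named reroutes.
# The SoloSnooze body's material function, shared with this backpack, feeds the FluffyFabric function, then the basic tech-art customization.
# A PostFX function finishes the chain, with shading model chosen by expression.
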Material /BRCosmetics/Accessories/FORT_Backpacks/Backpack_F_MED_SoloSnooze/Materials/M_SoloSnooze
graph {
 properties {
  bUseMaterialAttributes=true
  bUsedWithMeshDeformer=true
  bUsedWithMorphTargets=true
  bUsedWithSkeletalMesh=true
  pixelDepthOffsetMode=PDOM_Legacy
  shadingModel=MSM_FromMaterialExpression
  usedShadingModels=Default Lit | Subsurface | Clear Coat | Subsurface Profile | Cloth
 }
 nodes [
  n1 = MaterialFunctionCall!(Overide Texture Set UVs, Override Tech Art Mask UVs, Override Basic Emissive UVs) {materialFunction=/Game/Global/Materials/Character_Parent_2023/Functions/MF_CharacterParent_2023_Customization_PreFX}
  MaterialFunctionCall(
   Material=NamedRerouteUsage()
   M=NamedRerouteDeclaration(None=@n1.M) {name=M, nodeColor=<rgba 0.7344 1 0 1>}
   Pre Skinned Local Position=NamedRerouteDeclaration(None=@n1.Pre skinned Local Position) {name=PreskinnedPos, nodeColor=<rgba 0.7344 1 0 1>}
   Distance Blend=NamedRerouteDeclaration(None=@n1.Distance Blend) {name=DistanceBlend, nodeColor=<rgba 0.7344 1 0 1>}
   Fresnel=NamedRerouteDeclaration(None=@n1.Fresnel) {name=Fresnel, nodeColor=<rgba 0.7344 1 0 1>}
  ) {materialFunction=&MF_CharacterParent_2023_PostFX}
  NamedRerouteDeclaration(None=@n1.Result) {name=Material, nodeColor=<rgba 1 0.0177 0 1>}
  NamedRerouteDeclaration(None=@n1.Emissive UVs) {name=EmissiveUVs, nodeColor=<rgba 0.7344 1 0 1>}
  NamedRerouteDeclaration(None=@n1.Tech Art Mask) {name=TechArtMask, nodeColor=<rgba 0.7344 1 0 1>}
  n2 = NamedRerouteUsage()
  n3 = MaterialFunctionCall!(Material=NamedRerouteUsage(), Tech Art Mask=@n2) {materialFunction=/BRCosmetics/Characters/Player/Female/Medium/Bodies/F_MED_SoloSnooze/Materials/MF_SoloSnooze}
  NamedRerouteDeclaration(None=MaterialFunctionCall!(
    Tech Art Mask=NamedRerouteUsage()
    Emissive UVs=NamedRerouteUsage()
    LocalPosition=NamedRerouteUsage()
    Fresnel=NamedRerouteUsage()
    Mask Input=NamedRerouteUsage()
    Material=MaterialFunctionCall!(Material=@n3.Base Color, TechArtMask=@n2) {materialFunction=/Game/Global/Materials/Character_Parent_2023/FluffyFabric_Parent_2023/MF_FluffyFabric}.Result
   ) {materialFunction=/Game/Global/Materials/Character_Parent_2023/Functions/MF_CharacterParent_2023_CustomizationBasicTechArt}.Result) {name=Material_OUT, nodeColor=<rgba 0.0068 1 0.332 1>}
 ]
}
